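; Graffiti decal actor for HollowBase that removes itself when the building pieces it is attached to are destroyed.
; BeginPlay copies the decal material onto a texture streaming reference component and optionally overrides fade screen size by view distance quality, sharing one value below Epic.
; It prunes invalid parent walls and props, binds their OnDestroyed events, and destroys itself when a removal leaves either list empty.
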(event Collision|EventActorBeginOverlap
  (params (OtherActor :type Actor%20Object%20Reference)))

(event EventTick
  (params (DeltaSeconds :type Float%20%28single-precision%29))
  (|CallParentFunction DeltaSeconds) ; This node is disabled and will not be called.\nDrag off pins to build functionality.
)

(event EventBeginPlay
  (bind _decal (Variables|Decal|GetDecal))
  (Rendering|Material|SetMaterial (Variables|Default|GetTextureStreamingRef) 0 (Rendering|Components|Decal|GetDecalMaterial _decal))
  (CallFunction|OverrideFadeScreenSizePerQuality (Variables|Decal|GetOverride%20Per%20Quality%20Level?) (Variables|Decal|GetOverride%20-%20Fade%20Screen%20Size%20Below%20Epic) (Variables|Decal|GetOverride%20-%20Fade%20Screen%20Size%20Below%20Epic) (Variables|Decal|GetOverride%20-%20Fade%20Screen%20Size%20Below%20Epic) (Variables|/Script/Engine.DecalComponent|GetFadeScreenSize _decal))
  (Utilities|FlowControl|Sequence
    (:then_0
      ;[1 1 1] BuiuldingWall
      (bind (_arrayindex _arrayelement) (Utilities|Array|ReverseforEachLoop (Variables|Default|GetParentBuildingWalls))
        (:LoopBody
          (Utilities|IsValid _arrayelement
            (:Is%20Valid
              (Game|BindEventtoOnDestroyed _arrayelement (delegate Custom|ParentWallDestroyed%20)))
            (:Is%20Not%20Valid
              (Utilities|Array|RemoveItem (Variables|Default|GetParentBuildingWalls) _arrayelement))))))
    (:then_1
      ;[1 1 1] All BuildingActors
      (bind (_arrayindex_2 _arrayelement_2) (Utilities|Array|ReverseforEachLoop (Variables|Default|GetParentBuildingProps))
        (:LoopBody
          (Utilities|IsValid _arrayelement_2
            (:Is%20Valid
              (Game|BindEventtoOnDestroyed _arrayelement_2 (delegate Custom|ParentPropDestroyed)))
            (:Is%20Not%20Valid
              (Utilities|Array|RemoveItem (Variables|Default|GetParentBuildingProps) _arrayelement_2))))))))

(event Custom|ParentWallDestroyed%20
  (params (DestroyedActor :type Actor%20Object%20Reference))
  (bind _parentbuildingwalls (Variables|Default|GetParentBuildingWalls))
  ;[1 1 1] BuiuldingWall
  (bind _asparent_building_wall (Utilities|Casting|CastToParent_BuildingWall DestroyedActor))
  (Utilities|Array|RemoveItem _parentbuildingwalls _asparent_building_wall)
  (if (Utilities|Array|IsEmpty _parentbuildingwalls)
    (Actor|DestroyActor)))

(event Custom|ParentPropDestroyed
  (params (DestroyedActor :type Actor%20Object%20Reference))
  (bind _parentbuildingprops (Variables|Default|GetParentBuildingProps))
  ;[1 1 1] All BuildingActors
  (bind _asbuilding_actor (Utilities|Casting|CastToBuildingActor DestroyedActor))
  (Utilities|Array|RemoveItem _parentbuildingprops _asbuilding_actor)
  (if (Utilities|Array|IsEmpty _parentbuildingprops)
    (Actor|DestroyActor)))

(fn OverrideFadeScreenSizePerQuality
  (params (OverridePerQualityLevel? :type Boolean) (Low :type Float%20%28double-precision%29) (Medium :type Float%20%28double-precision%29) (High :type Float%20%28double-precision%29) (Epic :type Float%20%28double-precision%29))
  (bind _decal (Variables|Decal|GetDecal))
  (if OverridePerQualityLevel?
    (bind _returnvalue (Development|GetConsoleVariableIntValue "sg.ViewDistanceQuality"))
    (switch int _returnvalue
      (:Default
        (return false))
      (:0
        (Rendering|Components|Decal|SetFadeScreenSize _decal Low))
      (:1
        (Rendering|Components|Decal|SetFadeScreenSize _decal Medium))
      (:2
        (Rendering|Components|Decal|SetFadeScreenSize _decal High))
      (:3
        (Rendering|Components|Decal|SetFadeScreenSize _decal Epic)))
    (return true)
    (else
      (return false))))

(fn ConstructionScript
  (|CallParentFunction))
