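; Static mesh actor Blueprint for a configurable graffiti decal, using a Celebration-set runtime virtual texture (RVT) decal material.
; The construction script creates a dynamic instance of that material on mesh slot 0 and sets its Diffuse texture from a variable.
; It then applies the Decal Tint variable to both DecalColor parameters and the Opacity variable to its matching scalar across the mesh's materials.
; BeginPlay, overlap and Tick events are empty, so all setup happens at construction.
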(event EventBeginPlay)

(event Collision|EventActorBeginOverlap
  (params (OtherActor :type Actor%20Object%20Reference)))

(event EventTick
  (params (DeltaSeconds :type Float%20%28single-precision%29)))

(fn ConstructionScript
  (bind _decal_tint (Variables|Colors|GetDecal%20Tint))
  (bind _returnvalue (Math|Conversions|ToVector(LinearColor) _decal_tint))
  (bind _staticmeshcomponent (Variables|StaticMeshActor|GetStaticMeshComponent))
  (bind _returnvalue_2 (Rendering|Material|CreateDynamicMaterialInstance _staticmeshcomponent 0 "/Game/Environments/Artemis/Sets/Celebration/Materials/Master_Materials/M_RVT_Decal_Celebration_Graffiti.M_RVT_Decal_Celebration_Graffiti" "None" :!class /Script/Engine.PrimitiveComponent))
  (Rendering|Material|SetTextureParameterValue _returnvalue_2 "Diffuse" (Variables|Textures|GetDiffuse))
  (Rendering|Material|SetVectorParameterValueonMaterials _staticmeshcomponent "DecalColor" _returnvalue)
  (Rendering|Material|SetVectorParameterValueonMaterials _staticmeshcomponent "DecalColor2" _returnvalue)
  (Rendering|Material|SetScalarParameterValueonMaterials _staticmeshcomponent "Opacity" (Variables|Colors|GetOpacity)))
